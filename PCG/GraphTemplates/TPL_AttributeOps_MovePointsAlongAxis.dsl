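; PCG graph template that demonstrates moving points along their local up axis by a Distance parameter using attribute math.
; Three sample points in component-local space, two with tilted rotations, feed a multiply of each Rotation.Up by Distance, stored as DisplacementVector.
; A second math op combines that vector with Position to offset each point.
; Debug nodes display both the original and displaced points for comparison.
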(graph TPL_AttributeOps_MovePointsAlongAxis
  (params (Distance :type Double :default 2e+02))
  (bind CreatePoints_0 (Create%20Points :CoordinateSpace LocalComponent :PointsToCreate [(struct :Transform (struct :Rotation (0.0 -0.3826834323650898 0.0 0.9238795325112866)) :BoundsMin (-5e+01 -5e+01 -5e+01) :BoundsMax (5e+01 5e+01 5e+01) :Steepness 1.0) (struct :Transform (struct :Translation (0.0 2.5e+02 0.0)) :BoundsMin (-5e+01 -5e+01 -5e+01) :BoundsMax (5e+01 5e+01 5e+01) :Steepness 1.0) (struct :Transform (struct :Rotation (-0.0 0.3826834323650898 0.0 0.9238795325112866) :Translation (0.0 5e+02 0.0)) :BoundsMin (-5e+01 -5e+01 -5e+01) :BoundsMax (5e+01 5e+01 5e+01) :Steepness 1.0)]))
  (bind Debug_0 (Debug))
  (bind AttributeMathsOp_0 (Attribute%20Maths%20Op :InputSource1 $Rotation.Up :Operation Multiply :OutputTarget DisplacementVector :InB Distance))
  (bind AttributeMathsOp_1 (Attribute%20Maths%20Op :InputSource1 $Position :InputSource2 DisplacementVector))
  (bind Debug_1 (Debug))
  (connect CreatePoints_0 Out AttributeMathsOp_0 InA)
  (connect AttributeMathsOp_0 Out AttributeMathsOp_1 InA)
  (connect AttributeMathsOp_0 Out AttributeMathsOp_1 InB)
  (connect CreatePoints_0 Out Debug_0 In)
  (connect AttributeMathsOp_1 Out Debug_1 In))
