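; PCG template graph demonstrating attribute math on spline control points, previewed through Debug nodes.
; It uses the actor's spline, or builds a fallback from a jittered point grid via Branch and an execution dependency.
; An index-modulo attribute on control points carries into distance-based samples and is smoothstepped with math nodes.
; The result goes to color, then drives two outputs: points displaced along their right vector and points scaled in Y.
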(graph TPL_AttributeOps_SplineControlPoint
  (bind GetSplineData_0 (Get%20Spline%20Data))
  (bind DataNum_1 (Data%20Count))
  (bind Branch_1 (Branch))
  (bind SplineSampler_0 (Spline%20Sampler :SamplerParams (struct :Mode Distance :DistanceIncrement 1e+01 :NumSamples 50 :PointSteepness 1.0)))
  (bind CopyAttributes_0 (Copy%20Attributes :InputSource @Last :OutputTarget $Color))
  (bind AttributeMathsOp_14 (Attribute%20Maths%20Op :InputSource1 MyAttrib :Operation Subtract :OutputTarget Displacement))
  (bind AttributeMathsOp_3 (Attribute%20Maths%20Op :Operation Multiply))
  (bind AttributeMathsOp_20 (Attribute%20Maths%20Op :InputSource1 $Rotation.Right :Operation Multiply :OutputTarget Displacement))
  (bind AttributeMathsOp_21 (Attribute%20Maths%20Op :InputSource1 $Position))
  (bind Debug_0 (Debug))
  (bind AttributeMathsOp_16 (Attribute%20Maths%20Op :InputSource1 MyAttrib :Operation Multiply :OutputTarget Scale))
  (bind AttributeMathsOp_17 (Attribute%20Maths%20Op))
  (bind AttributeMathsOp_18 (Attribute%20Maths%20Op :InputSource1 $Scale.y :Operation Set))
  (bind Debug_1 (Debug))
  ;[0 0 0 0.5] Fallback Spline:
  ; -If no spline is found on the actor, the graph will create its own spline
  ; -if a spline is found, the branch A will be culled and not executed. note the use of the "Execution Dependency Required" advanced setting on the Create Points Grid node
  (bind CreatePointsGrid_1 (Create%20Points%20Grid :bSetPointsBounds false :CellSize (5e+02 1e+02 1e+02) :CoordinateSpace LocalComponent :GridExtents (2.5e+03 5e+01 5e+01)))
  (bind Branch_0 (Branch))
  (bind TransformPoints_0 (Transform%20Points :OffsetMax (1e+02 0.0 0.0) :OffsetMin (-1e+02 0.0 0.0)))
  (bind CreateSpline_1 (Create%20Spline :InterpTypeAttribute None :Mode CreateComponent))
  (bind Debug_2 (Debug))
  ;[0.15 0.15 0.15 0.5] Create an attribute on Control Points
  (bind AttributeMathsOp_2 (Attribute%20Maths%20Op :InputSource1 $Index :Operation Modulo :OutputTarget MyAttrib))
  ;[0.15 0.15 0.15 0.5] SmoothStep (assumes input is within 0-1 range)
  (bind AttributeMathsOp_10 (Attribute%20Maths%20Op :Operation Pow))
  (bind AttributeMathsOp_11 (Attribute%20Maths%20Op :Operation Multiply))
  (bind CreateConstant_0 (Create%20Constant :AttributeTypes (struct :DoubleValue 3.0)))
  (bind AttributeMathsOp_12 (Attribute%20Maths%20Op :Operation Subtract))
  (bind AttributeMathsOp_13 (Attribute%20Maths%20Op :Operation Multiply))
  (connect Branch_1 Output%20B AttributeMathsOp_2 InA)
  (connect CreateSpline_1 Out AttributeMathsOp_2 InA)
  (connect AttributeMathsOp_14 Out AttributeMathsOp_3 InA)
  (connect SplineSampler_0 Out AttributeMathsOp_10 InA)
  (connect SplineSampler_0 Out AttributeMathsOp_11 InA)
  (connect CreateConstant_0 Out AttributeMathsOp_12 InA)
  (connect AttributeMathsOp_11 Out AttributeMathsOp_12 InB)
  (connect AttributeMathsOp_10 Out AttributeMathsOp_13 InA)
  (connect AttributeMathsOp_12 Out AttributeMathsOp_13 InB)
  (connect CopyAttributes_0 Out AttributeMathsOp_14 InA)
  (connect CopyAttributes_0 Out AttributeMathsOp_16 InA)
  (connect AttributeMathsOp_16 Out AttributeMathsOp_17 InA)
  (connect AttributeMathsOp_17 Out AttributeMathsOp_18 InA)
  (connect AttributeMathsOp_17 Out AttributeMathsOp_18 InB)
  (connect AttributeMathsOp_3 Out AttributeMathsOp_20 InA)
  (connect AttributeMathsOp_3 Out AttributeMathsOp_20 InB)
  (connect AttributeMathsOp_20 Out AttributeMathsOp_21 InA)
  (connect AttributeMathsOp_20 Out AttributeMathsOp_21 InB)
  (connect CreatePointsGrid_1 Out Branch_0 In)
  (connect DataNum_1 Out Branch_1 bOutputToB)
  (connect GetSplineData_0 Out Branch_1 In)
  (connect AttributeMathsOp_13 Out CopyAttributes_0 Source)
  (connect AttributeMathsOp_13 Out CopyAttributes_0 Target)
  (connect Branch_1 Output%20A CreatePointsGrid_1 Execution%20Dependency)
  (connect TransformPoints_0 Out CreateSpline_1 In)
  (connect GetSplineData_0 Out DataNum_1 In)
  (connect AttributeMathsOp_21 Out Debug_0 In)
  (connect AttributeMathsOp_18 Out Debug_1 In)
  (connect TransformPoints_0 Out Debug_2 In)
  (connect AttributeMathsOp_2 Out SplineSampler_0 Spline)
  (connect Branch_0 Output%20A TransformPoints_0 In))
