(graph TPL_Sampler_CollisionData
  (params (StaticMesh :type (Object /Script/Engine.StaticMesh) :default /Engine/EngineMeshes/SM_MatPreviewMesh_01.SM_MatPreviewMesh_01))
  (bind CreatePoints_1 (Create%20Points))
  (bind AddAttribute_1 (Add%20Attribute :OutputTarget Mesh :Attributes StaticMesh))
  (bind BoundsFromMesh_0 (Bounds%20From%20Mesh))
  (bind CreatePointsGrid_0 (Create%20Points%20Grid :bSetPointsBounds false :CellSize (1e+03 1e+03 1e+02) :CoordinateSpace LocalComponent :GridExtents (2.5e+03 2.5e+03 5e+01)))
  (bind TransformPoints_0 (Transform%20Points :RotationMax (struct :Yaw 3.6e+02) :ScaleMax (4.0 4.0 4.0) :ScaleMin (2.0 2.0 2.0)))
  (bind CopyPoints_0 (Copy%20Points))
  (bind CreateCollisionData_0 (Create%20Collision%20Data :CollisionAttribute Mesh :CollisionQueryFlag Complex))
  (bind VolumeSampler_0 (Volume%20Sampler :VoxelSize (5e+01 5e+01 5e+01)))
  (bind AttributeReduce_0 (Attribute%20Reduce :InputSource $Position.z :Operation Min))
  (bind AttributeReduce_1 (Attribute%20Reduce :InputSource $Position.z :Operation Max))
  (bind AttributeRemap_0 (Attribute%20Remap :InputSource $Position.z :OutputTarget $Color))
  (bind Debug_2 (Debug))
  (connect CreatePoints_1 Out AddAttribute_1 In)
  (connect VolumeSampler_0 Out AttributeReduce_0 In)
  (connect VolumeSampler_0 Out AttributeReduce_1 In)
  (connect VolumeSampler_0 Out AttributeRemap_0 In)
  (connect AttributeReduce_1 Out AttributeRemap_0 InRangeMax)
  (connect AttributeReduce_0 Out AttributeRemap_0 InRangeMin)
  (connect AddAttribute_1 Out BoundsFromMesh_0 In)
  (connect BoundsFromMesh_0 Out CopyPoints_0 Source)
  (connect TransformPoints_0 Out CopyPoints_0 Target)
  (connect CopyPoints_0 Out CreateCollisionData_0 In)
  (connect AttributeRemap_0 Out Debug_2 In)
  (connect CreatePointsGrid_0 Out TransformPoints_0 In)
  (connect CreateCollisionData_0 Out VolumeSampler_0 Volume))
