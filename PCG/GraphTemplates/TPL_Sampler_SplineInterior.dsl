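; A PCG graph template that scatters points across the interior area enclosed by a spline.
; Get Spline Data supplies the spline to a Spline Sampler set to sample on the interior by distance, with full point steepness.
; The sampled points feed a Debug node for visualization, leaving the graph as a starting point with no further processing or spawning.
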(graph TPL_Sampler_SplineInterior
  (bind GetSplineData_1 (Get%20Spline%20Data))
  (bind SplineSampler_1 (Spline%20Sampler :SamplerParams (struct :Dimension OnInterior :Mode Distance :NumSamples 50 :PointSteepness 1.0)))
  (bind Debug_3 (Debug))
  (connect SplineSampler_1 Out Debug_3 In)
  (connect GetSplineData_1 Out SplineSampler_1 Spline))
