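; PCG graph template that demonstrates two ways of sampling a tiling noise texture, each shown through a Debug node.
; A local-component point grid passes through a Position.x math op into Sample Texture.
; The sampled red channel is multiplied by the Displacement parameter and applied to each point's height, forming a heightfield.
; In a separate branch, Surface Sampler uses the same texture data directly as its surface to scatter points.
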(graph TPL_Sampler_Texture
  (params (Displacement :type Double :default 7.5e+02))
  (bind CreatePointsGrid_0 (Create%20Points%20Grid :CellSize (16.0 16.0 64.0) :CoordinateSpace LocalComponent :GridExtents (2.56e+03 2.56e+03 32.0)))
  (bind AttributeMathsOp_0 (Attribute%20Maths%20Op :InputSource1 $Position.x))
  (bind GetTextureData_0 (Get%20Texture%20Data :bUseAdvancedTiling true :Texture /Engine/MapTemplates/TilingNoise05.TilingNoise05))
  (bind SampleTexture_0 (Sample%20Texture))
  (bind SurfaceSampler_0 (Surface%20Sampler :PointsPerSquaredMeter 0.5))
  (bind Debug_0 (Debug))
  (bind AttributeMathsOp_1 (Attribute%20Maths%20Op :InputSource1 $Color.r :Operation Multiply :OutputTarget Displacement :InB Displacement))
  (bind AttributeMathsOp_2 (Attribute%20Maths%20Op :InputSource1 $Position.z :InputSource2 Displacement))
  (bind Debug_1 (Debug))
  (connect CreatePointsGrid_0 Out AttributeMathsOp_0 InA)
  (connect SampleTexture_0 Out AttributeMathsOp_1 InA)
  (connect AttributeMathsOp_1 Out AttributeMathsOp_2 InA)
  (connect AttributeMathsOp_1 Out AttributeMathsOp_2 InB)
  (connect SurfaceSampler_0 Out Debug_0 In)
  (connect AttributeMathsOp_2 Out Debug_1 In)
  (connect GetTextureData_0 Out SampleTexture_0 BaseTexture)
  (connect AttributeMathsOp_0 Out SampleTexture_0 Point)
  (connect GetTextureData_0 Out SurfaceSampler_0 Surface))
