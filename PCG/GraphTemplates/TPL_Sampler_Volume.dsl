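; TPL_Sampler_Volume is a PCG graph template that samples points inside a volume and visualizes them.
; Get Volume Data feeds the Volume input of a Volume Sampler, whose output goes to a Debug node.
; A StaticMesh graph parameter defaults to the engine's material preview mesh, but no connection in the graph uses it.
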(graph TPL_Sampler_Volume
  (params (StaticMesh :type (Object /Script/Engine.StaticMesh) :default /Engine/EngineMeshes/SM_MatPreviewMesh_01.SM_MatPreviewMesh_01))
  (bind GetVolumeData_0 (Get%20Volume%20Data))
  (bind VolumeSampler_1 (Volume%20Sampler))
  (bind Debug_0 (Debug))
  (connect VolumeSampler_1 Out Debug_0 In)
  (connect GetVolumeData_0 Out VolumeSampler_1 Volume))
