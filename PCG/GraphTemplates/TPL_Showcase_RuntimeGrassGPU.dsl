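; PCG template graph that showcases runtime GPU grass scattering on a landscape, using sample seedling meshes as the foliage.
; Landscape data passes through a Grid16 hierarchical generation step into a Custom HLSL point generator with a fixed count of 262,144 elements.
; The points feed a weighted Static Mesh Spawner with two material variants of one seedling.
; Graph parameters expose density, random offset, scale and noise controls; the kernel source is absent here.
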(graph TPL_Showcase_RuntimeGrassGPU
  (params (DensityFilter :type Double :default 1.0) (RandomNoiseScale :type Double :default 1e+02) (RandomPositionOffset :type Double :default 1e+02) (PointsScaleMin :type Double :default 0.7) (PointsScaleMax :type Double :default 1.0) (SpatialNoiseSize :type Double :default 125.0) (SpatialNoiseStrength :type Double :default 0.45) (ScaleByDensityMin :type Double :default 0.05) (ScaleByDensityMax :type Double :default 0.3) (RandomNoiseSeed :type Int :default 42) (SpatialNoiseSeed :type Int :default 42))
  (bind GetLandscapeData_0 (Get%20Landscape%20Data :SamplingProperties (struct)))
  (bind HiGenGridSize_1 (Change%20Grid%20Size :HiGenGridSize Grid16))
  (bind CustomHLSL_0 (Custom%20HLSL :InputPins [(struct :Label Landscape :AllowedTypes (struct :Ids [(struct :Struct /Script/PCG.PCGDataTypeInfoLandscape)]) :bAllowMultipleData false :bAllowMultipleConnections false) (struct :Label GraphParams :AllowedTypes (struct :Ids [(struct :Struct /Script/PCG.PCGDataTypeInfoParam)]) :bAllowMultipleData false :PinStatus Required :bAllowMultipleConnections false)] :KernelType PointGenerator :NumElements 262144 :GraphParams ))
  (bind StaticMeshSpawner_1 (Static%20Mesh%20Spawner :MeshSelectorParameters (class:/Script/PCG.PCGMeshSelectorWeighted :MeshEntries [(struct :Descriptor (struct :StaticMesh /PCG/SampleContent/SimpleForest/Meshes/PCG_Seedling_01.PCG_Seedling_01 :OverrideMaterials [/PCG/SampleContent/SimpleForest/Materials/PCG_Foliage_01.PCG_Foliage_01] :bCanEverAffectNavigation true)) (struct :Descriptor (struct :StaticMesh /PCG/SampleContent/SimpleForest/Meshes/PCG_Seedling_01.PCG_Seedling_01 :OverrideMaterials [/PCG/SampleContent/SimpleForest/Materials/PCG_Foliage_01_Inst.PCG_Foliage_01_Inst] :bCanEverAffectNavigation true))])))
  (connect HiGenGridSize_1 Out CustomHLSL_0 Landscape)
  (connect GetLandscapeData_0 Out HiGenGridSize_1 In)
  (connect CustomHLSL_0 Out StaticMeshSpawner_1 In))
