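; PCG showcase template that uses shape grammar to place fence mesh modules along a spline.
; Module symbols are paired with meshes, taking size and pivot offset from mesh bounds, then feed Subdivide Spline with the Grammar string.
; The X pivot offset is scaled, rotated and applied before a toggle selects static or spline mesh spawning.
; Without an actor spline, an execution-dependency branch creates a fallback spline from projected points.
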(graph TPL_Showcase_ShapeGrammar
  (params (ModuleInfo :type (Array (Struct /Script/PCG.PCGSubdivisionSubmodule)) :default [(struct :Symbol Large1 :bScalable true :DebugColor (1.0 0.0 0.0 1.0)) (struct :Symbol Large2 :bScalable true :DebugColor (0.0 1.0 0.0 1.0)) (struct :Symbol Small1 :bScalable true :DebugColor (0.0 0.0 1.0 1.0)) (struct :Symbol Large3 :bScalable true :DebugColor (1.0 0.0 1.0 1.0))]) (MeshInfo :type (Array (Object /Script/Engine.StaticMesh)) :default [/PCG/GraphTemplates/Meshes/Fence01.Fence01 /PCG/GraphTemplates/Meshes/Fence02.Fence02 /PCG/GraphTemplates/Meshes/Fence03.Fence03 /PCG/GraphTemplates/Meshes/Fence04.Fence04]) (Grammar :type String :default "[Small1]{Large3,Large1}*[Large2]") (SpawnSplineMeshes :type Bool))
  (inputs (NewPin :type Spline))
  (bind AddAttribute_2 (Add%20Attribute :OutputTarget Mesh :Attributes MeshInfo :In ModuleInfo))
  (bind GetSplineData_1 (Get%20Spline%20Data))
  (bind DataNum_2 (Data%20Count))
  (bind Branch_2 (Branch))
  (bind AttributeSetToPoint_0 (Attribute%20Set%20To%20Point))
  (bind BoundsFromMesh_1 (Bounds%20From%20Mesh :MeshAttribute Mesh))
  (bind AttributeMathsOp_18 (Attribute%20Maths%20Op :InputSource1 Size :InputSource2 $LocalSize.X :Operation Set)) ; LocalSize.X\nfor module size
  (bind AttributeMathsOp_17 (Attribute%20Maths%20Op :InputSource1 $LocalCenter :Operation Negate :OutputTarget PivotOffset)) ; -LocalCenter is the PivotOffset
  (bind PointToAttributeSet_0 (Point%20To%20Attribute%20Set))
  (bind SubdivideSpline_1 (Subdivide%20Spline :bAcceptIncompleteSubdivision true :bForwardAttributesFromModulesInfo true :bModuleInfoAsInput true :bOutputScalableAttribute false :bOutputSizeAttribute false :GrammarSelection (struct :GrammarString "[C]{W,W1}*[C]") :ModuleHeight 0.0 :ModuleHeightAttribute ExtentsZ :ModulesInfoAttributeNames (struct :bProvideScalable true :ScalableAttributeName bScalable) :ScalableAttributeName bScalable :GrammarString Grammar)) ; Module info is \nembedded in the node
  (bind Branch_0 (Branch :bOutputToB-param SpawnSplineMeshes))
  (bind CreateSpline_9 (Create%20Spline :ArriveTangentAttribute None :InterpTypeAttribute None :LeaveTangentAttribute None))
  (bind SpawnSplineMesh_1 (Spawn%20Spline%20Mesh :SplineMeshOverrideDescriptions [(struct :InputSource Mesh :PropertyTarget "StaticMesh")]))
  (bind StaticMeshSpawner_0 (Static%20Mesh%20Spawner :MeshSelectorParameters (class:/Script/PCG.PCGMeshSelectorByAttribute :AttributeName Mesh) :MeshSelectorType /Script/PCG.PCGMeshSelectorByAttribute))
  ;[0.15 0.15 0.15 0.5] Apply Mesh Pivot Offset
  (bind CreateConstant_0 (Create%20Constant :AttributeTypes (struct :Type Vector :VectorValue (1.0 0.0 0.0))))
  (bind AttributeMathsOp_10 (Attribute%20Maths%20Op :InputSource1 PivotOffset :Operation Multiply)) ; Only Consider Offset X (Along Spline)
  (bind AttributeMathsOp_0 (Attribute%20Maths%20Op :InputSource1 PivotOffset :InputSource2 $Scale :Operation Multiply)) ; PivotOffset needs to be \nscaled too
  (bind AttributeVectorOp_0 (Attribute%20Vector%20Op :InputSource1 PivotOffset :InputSource2 $Transform :Operation TransformDirection)) ; And rotated
  (bind AttributeMathsOp_3 (Attribute%20Maths%20Op :InputSource1 $Position :InputSource2 PivotOffset)) ; And added to the position
  ;[0 0 0 0.5] Fallback Spline:
  ; -If no spline is found on the actor, the graph will create its own spline
  ; -if a spline is found, the branch A will be culled and not executed. note the use of the "Execution Dependency Required" advanced setting on the Create Points and WRHQ nodes
  (bind CreatePoints_0 (Create%20Points :CoordinateSpace LocalComponent :PointsToCreate [(struct :Transform (struct :Translation (-2.5e+03 1e+03 0.0)) :Steepness 1.0) (struct :Transform (struct :Translation (-5e+02 1e+03 0.0)) :Steepness 1.0) (struct :Steepness 1.0) (struct :Transform (struct :Translation (2e+03 0.0 0.0)) :Steepness 1.0)]))
  (bind WorldRayHitQuery_0 (World%20Ray%20Hit%20Query))
  (bind Projection_0 (Projection))
  (bind CreateSpline_0 (Create%20Spline :InterpTypeAttribute None :Mode CreateComponent))
  (connect AttributeMathsOp_10 Out AttributeMathsOp_0 InA)
  (connect AttributeMathsOp_10 Out AttributeMathsOp_0 InB)
  (connect AttributeVectorOp_0 Out AttributeMathsOp_3 InA)
  (connect AttributeVectorOp_0 Out AttributeMathsOp_3 InB)
  (connect SubdivideSpline_1 Out AttributeMathsOp_10 InA)
  (connect CreateConstant_0 Out AttributeMathsOp_10 InB)
  (connect AttributeMathsOp_18 Out AttributeMathsOp_17 In)
  (connect BoundsFromMesh_1 Out AttributeMathsOp_18 InA)
  (connect BoundsFromMesh_1 Out AttributeMathsOp_18 InB)
  (connect AddAttribute_2 Out AttributeSetToPoint_0 In)
  (connect AttributeMathsOp_0 Out AttributeVectorOp_0 In)
  (connect AttributeMathsOp_0 Out AttributeVectorOp_0 Transform)
  (connect AttributeSetToPoint_0 Out BoundsFromMesh_1 In)
  (connect AttributeMathsOp_3 Out Branch_0 In)
  (connect DataNum_2 Out Branch_2 bOutputToB)
  (connect GetSplineData_1 Out Branch_2 In)
  (connect Branch_2 Output%20A CreatePoints_0 Execution%20Dependency)
  (connect Projection_0 Out CreateSpline_0 In)
  (connect Branch_0 Output%20B CreateSpline_9 In)
  (connect GetSplineData_1 Out DataNum_2 In)
  (connect AttributeMathsOp_17 Out PointToAttributeSet_0 In)
  (connect CreatePoints_0 Out Projection_0 In)
  (connect WorldRayHitQuery_0 Out Projection_0 Projection%20Target)
  (connect CreateSpline_9 Out SpawnSplineMesh_1 In)
  (connect Branch_0 Output%20A StaticMeshSpawner_0 In)
  (connect Branch_2 Output%20B SubdivideSpline_1 In)
  (connect CreateSpline_0 Out SubdivideSpline_1 In)
  (connect PointToAttributeSet_0 Out SubdivideSpline_1 ModulesInfo)
  (connect Branch_2 Output%20A WorldRayHitQuery_0 Execution%20Dependency))
