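; _Default_EmptyGraph is an empty PCG graph, stored among the graph templates, that by its name serves as the default blank starting point for new graphs.
; It defines no nodes, connections or parameters, so it generates nothing by itself.
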(graph _Default_EmptyGraph)
