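; _Default_Loop is a PCG graph in the GraphTemplates folder, named as a default template for loop graphs.
; It shows no nodes, connections, parameters or non-default settings, so it performs no processing of its own.
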(graph _Default_Loop)
