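; Default PCG graph template for a loop subgraph that carries feedback data between iterations.
; It declares a Feedback input pin and a matching Feedback output pin, both accepting any data type.
; The body is empty, with no nodes or links between the two pins, so it serves only as a starting skeleton.
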(graph _Default_LoopFeedback
  (inputs (Feedback :type Any))
  (outputs (Feedback :type Any)))
